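# One-shot stereo MetaSound source that plays the PulseLancer UI sound for Hunter Rank level benefits.
# A single stereo Wave Player starts on play, feeds both output channels and signals finished when the wave ends.
# It routes through the non-critical in-game UI sound class with a raised volume setting.
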MetaSoundSource /Game/Athena/Sounds/LTMs/SMASH/SFX/IP_MSS_PulseLancer_UI_HunterRank_LevelBenifits
inputs [UE.Source.OnPlay {type=Trigger}]
interfaces [UE.OutputFormat.Stereo:1.0, UE.Source.OneShot:1.0, UE.Source:1.1]
outputs [
 "UE.OutputFormat.Stereo.Audio:0" {type=Audio}
 "UE.OutputFormat.Stereo.Audio:1" {type=Audio}
 UE.Source.OneShot.OnFinished {type=Trigger}
]
pages [Default {
  input_defaults {UE.Source.OnPlay=null}
  nodes [WavePlayer_0 = UE.Wave Player:Stereo(Play=^UE.Source.OnPlay) {also {defaults {Wave Asset=&/IP_PulseLancer_UI_LevelBenifits}, version [1,1]}}]
  links [
   {dst="@UE.Source.OneShot.OnFinished", src=WavePlayer_0.On Finished}
   {dst="@UE.OutputFormat.Stereo.Audio:0", src=WavePlayer_0.Out Left}
   {dst="@UE.OutputFormat.Stereo.Audio:1", src=WavePlayer_0.Out Right}
  ]
 }]
properties {
 assetImportData=&self:AssetImportData
 numChannels=2
 outputFormat=Stereo
 sampleRate=48000
 soundClassObject=&Fort_UI_InGame_NonCritical
 volume=2
}
