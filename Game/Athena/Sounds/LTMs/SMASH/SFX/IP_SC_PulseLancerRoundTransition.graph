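# Sound cue for the round transition UI sting in the SMASH LTM, using Pulse Lancer audio.
# A fixed one-second delay precedes a mixer that layers the main transition wave with a ticks wave, the ticks weighted louder.
# The cue plays in the critical in-game UI sound class with boosted volume and overrides concurrency to allow at most two instances.
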SoundCue /Game/Athena/Sounds/LTMs/SMASH/SFX/IP_SC_PulseLancerRoundTransition
root=Delay_1
nodes [
 {id=Delay_1, children [Mixer_0], properties {delayMax=1, delayMin=1}}
 {id=Mixer_0, children [WavePlayer_4,WavePlayer_1], properties {inputVolume [1,2.5]}}
 {id=WavePlayer_1, properties {soundWaveAssetPtr=&/IP_PulseLancer_UI_RoundTransition_Ticks}}
 {id=WavePlayer_4, properties {soundWaveAssetPtr=&/IP_PulseLancer_UI_RoundTransition}}
]
properties {
 bOverrideConcurrency=true
 maxDistance=2097152
 soundClassObject=&Fort_UI_InGame_Critical
 volumeMultiplier=2
 concurrencyOverrides {maxCount=2}
}
