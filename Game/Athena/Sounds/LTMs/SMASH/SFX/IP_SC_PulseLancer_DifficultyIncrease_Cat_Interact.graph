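# Sound cue for the SMASH LTM Pulse Lancer, apparently played on a cat statue interaction that increases difficulty.
# A Random node picks one of two equally weighted cat statue plant waves, behind a fixed half-second delay and a volume-boosting Modulator with unchanged pitch.
# The cue overrides attenuation, limits concurrency to a single instance and routes to the critical in-game UI sound class.
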SoundCue /Game/Athena/Sounds/LTMs/SMASH/SFX/IP_SC_PulseLancer_DifficultyIncrease_Cat_Interact
root=Modulator_0
nodes [
 {id=Delay_0, children [Random_0], properties {delayMax=0.5, delayMin=0.5}}
 {id=Modulator_0, children [Delay_0], properties {pitchMax=1, pitchMin=1, volumeMax=3, volumeMin=3}}
 {id=Random_0, children [WavePlayer_0,WavePlayer_1], properties {weights [1,1]}}
 {id=WavePlayer_0, properties {soundWaveAssetPtr=&/IP_PulseLancer_CatStatue_Plant_01}}
 {id=WavePlayer_1, properties {soundWaveAssetPtr=&/IP_PulseLancer_CatStatue_Plant_02}}
]
properties {
 bOverrideAttenuation=true
 bOverrideConcurrency=true
 maxDistance=3000
 soundClassObject=&Fort_UI_InGame_Critical
 volumeMultiplier=2
 attenuationOverrides {falloffDistance=2000, attenuationShapeExtents {x=1000}}
 concurrencyOverrides {maxCount=1}
}
