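# IP_SC_PulseLancer_UI_HunterLevelBenifits is a SoundCue for a critical in-game UI sound in the SMASH LTM, named for Pulse Lancer hunter level benefits.
# A single wave player feeds the IP_PulseLancer_UI_LevelBenifits wave through a Modulator with fixed pitch and volume, so playback has no variation.
# The cue uses the critical in-game UI sound class with boosted volume.
# Its concurrency override allows one instance and blocks new plays while it sounds.
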SoundCue /Game/Athena/Sounds/LTMs/SMASH/SFX/IP_SC_PulseLancer_UI_HunterLevelBenifits
root=Modulator_0
nodes [
 {id=Modulator_0, children [WavePlayer_0], properties {pitchMax=1, pitchMin=1, volumeMax=1, volumeMin=1}}
 {id=WavePlayer_0, properties {soundWaveAssetPtr=&/IP_PulseLancer_UI_LevelBenifits}}
]
properties {
 bOverrideConcurrency=true
 maxDistance=2097152
 soundClassObject=&Fort_UI_InGame_Critical
 volumeMultiplier=2
 concurrencyOverrides {maxCount=1, resolutionRule=PreventNew}
}
